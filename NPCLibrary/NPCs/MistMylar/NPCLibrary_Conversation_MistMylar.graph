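# Conversation database for the MistMylar NPC (Daigo), providing his greeting and main conversation menu in the NPC Library.
# An entry point feeds one Speech node that picks a contextual line: a first-ever introduction, lines gated by match conversation count, or higher-priority reactions to specific player character skins, with a default fallback.
# The speech then branches through link nodes to shared quest, basic services, MistMylar dialogue and Season 37 story conversations.
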ConversationDatabase /NPCLibrary/NPCs/MistMylar/NPCLibrary_Conversation_MistMylar
entry_points [{tag=Conversation.Entry.MistMylar, destinations [#1]}]
nodes [
 ConversationEntryPointNode_7 {
  id=#1
  class=ConversationEntryPointNode
  outputs [#2]
  properties {entryTag {tagName=Conversation.Entry.MistMylar}}
 }
 Speech_C_4 {
  id=#2
  class=Speech_C
  outputs [#3, #4, #5, #6]
  properties {
   nodeName=""
   generalConfig {
    defaultMessage=I almost wish I HAD betrayed the Island. That would have been... less embarassing.
    contextualMessages [
     {
      message=If you somehow don't know who I am... My name is Daigo.
      contextRequirements [{requirement {class=FortControllerRequirement_IsFirstEverConversationWithNPC}}]
      priority {curve {rowName=Tandem.Speech.Priority.FirstEverInteraction}}
     }
     {
      message="It's... humiliating, that the entire Island knows the mistakes I've made."
      contextRequirements [{requirement {class=FortControllerRequirement_CheckNPCMatchConversationCount, properties {nPCUniqueIdTag {tagName=AISpawnerData.Type.Tandem.MistMylar}}}}]
      priority {curve {curveTable=null, rowName=None}}
     }
     {
      message="If you care about this Island, be ready to defend it with everything you've got."
      contextRequirements [{requirement {class=FortControllerRequirement_CheckNPCMatchConversationCount, properties {nPCUniqueIdTag {tagName=AISpawnerData.Type.Tandem.MistMylar}}}}]
      priority {curve {curveTable=null, rowName=None}}
     }
     {
      message="I tried to save the Island. I failed. So... now, I'll help how I can."
      contextRequirements [{requirement {class=FortControllerRequirement_CheckNPCMatchConversationCount, properties {nPCUniqueIdTag {tagName=AISpawnerData.Type.Tandem.MistMylar}}}}]
      priority {curve {curveTable=null, rowName=None}}
     }
     {
      message="I'm the smartest person on this Island, and still... made a huge mistake."
      contextRequirements [{requirement {class=FortControllerRequirement_CheckNPCMatchConversationCount, properties {nPCUniqueIdTag {tagName=AISpawnerData.Type.Tandem.MistMylar}}}}]
      priority {curve {curveTable=null, rowName=None}}
     }
     {
      message=I will never let myself become a puppet for the Spirit Realm again.
      contextRequirements [{requirement {class=FortControllerRequirement_CheckNPCMatchConversationCount, properties {nPCUniqueIdTag {tagName=AISpawnerData.Type.Tandem.MistMylar}}}}]
      priority {curve {curveTable=null, rowName=None}}
     }
     {
      message=I almost wish I HAD betrayed the Island. That would have been... less embarassing.
      contextRequirements [{requirement {class=FortControllerRequirement_CheckNPCMatchConversationCount, properties {nPCUniqueIdTag {tagName=AISpawnerData.Type.Tandem.MistMylar}}}}]
      priority {curve {curveTable=null, rowName=None}}
     }
     {
      message=If you have it in your heart to forgive my recklessness... I'm sorry.
      contextRequirements [{requirement {
         class=FortControllerRequirement_HasCID
         properties {allowedCIds [
           &Character_CubicVice_Dual
           /BRCosmetics/Athena/Items/Cosmetics/Characters/Character_PolishedJade_Mind
          ]}
        }}]
      priority {value=2, curve {curveTable=null, rowName=None}}
     }
     {
      message="That weight I felt, to protect everyone... I almost let it destroy me. Be careful."
      contextRequirements [{requirement {
         class=FortControllerRequirement_HasCID
         properties {allowedCIds [
           &Character_IronLilac_Sly
           &Character_BrightDisk
           &Character_VoidRedemption_Rebel
          ]}
        }}]
      priority {value=2, curve {curveTable=null, rowName=None}}
     }
     {
      message=Why should anyone entrust the Shard to someone like you?
      contextRequirements [{requirement {
         class=FortControllerRequirement_HasCID
         properties {allowedCIds [
           &Character_UndergroundRebel_Fashion
           &CID_989_Athena_Commando_M_ProgressiveJonesy
           &Character_AuraCop_Bullet
          ]}
        }}]
      priority {value=2, curve {curveTable=null, rowName=None}}
     }
     {
      message=I'm willing to do whatever it takes to save this Island. Even... *sigh* follow orders.
      contextRequirements [{requirement {class=FortControllerRequirement_HasCID, properties {allowedCIds [/BRCosmetics/Athena/Items/Cosmetics/Characters/Character_SwoopClaspPry]}}}]
      priority {value=2, curve {curveTable=null, rowName=None}}
     }
     {
      message=Do you still dream of the Spirit Realm? Of what it made you become?
      contextRequirements [{requirement {class=FortControllerRequirement_HasCID, properties {allowedCIds [&Character_RoseDepth_Seed]}}}]
      priority {value=2, curve {curveTable=null, rowName=None}}
     }
    ]
   }
  }
 }
 ConversationLinkNode_3 {
  id=#6
  class=ConversationLinkNode
  properties {remoteEntryTag {tagName=Conversation.Entry.Season37.Story.P03.Q01.C}}
 }
 ConversationLinkNode_2 {id=#5, class=ConversationLinkNode, properties {remoteEntryTag {tagName=Conversation.Entry.MistMylar.Dialogue}}}
 ConversationLinkNode_0 {id=#4, class=ConversationLinkNode, properties {remoteEntryTag {tagName=Conversation.Entry.BasicServices}}}
 ConversationLinkNode_1 {id=#3, class=ConversationLinkNode, properties {remoteEntryTag {tagName=Conversation.Entry.NPCLibrary.Quests}}}
]
